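# One-shot mono MetaSound source for weapon equip foley, using weapon foley attenuation, sound class and concurrency.
# On play, it picks a random wave from Sounds and plays it, then signals finished when the wave ends.
# Character affiliation selects among four float gains via a crossfade, which is multiplied with MasterVolume to scale the output.
# The Default page randomizes pitch around a base by a spread; the Mobile page omits pitch shifting.
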MetaSoundSource /FortMetasound/Weapons/MSS_Weapon_Equip
inputs [
 Character.Affiliation {type=Enum:FortAudioAffiliation}
 Character.Health {type=Float}
 Character.UniqueID {type=Int32}
 MasterVolume {type=Float}
 Random.PitchShift.Base {type=Float}
 Random.PitchShift.Spread {type=Float}
 Sounds {type=WaveAsset:Array}
 UE.Source.OnPlay {type=Trigger}
]
interfaces [Fort.Character:1.0, UE.OutputFormat.Mono:1.0, UE.Source.OneShot:1.0, UE.Source:1.1]
outputs [
 "UE.OutputFormat.Mono.Audio:0" {type=Audio}
 UE.Source.OneShot.OnFinished {type=Trigger}
]
pages [
 Default {
  is_build_graph=true
  input_defaults {
   Character.Affiliation=0
   Character.Health=0
   Character.UniqueID=0
   MasterVolume=1
   Random.PitchShift.Base=0
   Random.PitchShift.Spread=0.5
   UE.Source.OnPlay=null
   Sounds []
  }
  nodes [
   CrossfadeTriggerRouteFloat4_0 = "Crossfade.Trigger Route (Float, 4):Float"(
    Crossfade Value=Convert.Float:Int32(Int32=Convert.Int32:Enum:FortAudioAffiliation("Enum:FortAudioAffiliation"=^Character.Affiliation).Int32).Float
    In 0=VariableAccessor.Float(Variable=InitVariable.Float().Variable).Value
    In 1=VariableAccessor.Float(Variable=InitVariable.Float().Variable).Value
    In 2=VariableAccessor.Float(Variable=InitVariable.Float().Variable).Value
    In 3=VariableAccessor.Float(Variable=InitVariable.Float().Variable).Value
   )
   ArrayRandomGet_0 = Array.Random Get:WaveAsset:Array(In Array=^Sounds, Next=^UE.Source.OnPlay) {also {defaults {Enable Shared State="<MetaSoundBuilderResult.SUCCEEDED: 0>", No Repeats=-1, Weights []}}}
   WavePlayer_0 = UE.Wave Player:Mono(
    Play=@ArrayRandomGet_0.On Next
    Wave Asset=@ArrayRandomGet_0.Value
    Pitch Shift=UE.RandomFloat(
     Max=UE.Add:Float(PrimaryOperand=^Random.PitchShift.Base, AdditionalOperands=^Random.PitchShift.Spread).Out
     Next=@ArrayRandomGet_0.On Next
     Min=UE.Subtract:Float(PrimaryOperand=^Random.PitchShift.Base, AdditionalOperands=^Random.PitchShift.Spread).Out
    ).Value
   ) {also {version [1,1]}}
   Multiply_0 = UE.Multiply:Audio by Float(
    AdditionalOperands=UE.Multiply:Float(PrimaryOperand=^MasterVolume, AdditionalOperands=@CrossfadeTriggerRouteFloat4_0.Out).Out
    PrimaryOperand=@WavePlayer_0.Out Mono
   )
  ]
  links [{dst="@UE.OutputFormat.Mono.Audio:0", src=Multiply_0.Out}, {dst="@UE.Source.OneShot.OnFinished", src=WavePlayer_0.On Finished}]
 }
 Mobile {
  nodes [
   CrossfadeTriggerRouteFloat4_0 = "Crossfade.Trigger Route (Float, 4):Float"(
    Crossfade Value=Convert.Float:Int32(Int32=Convert.Int32:Enum:FortAudioAffiliation("Enum:FortAudioAffiliation"=^Character.Affiliation).Int32).Float
    In 0=VariableAccessor.Float(Variable=InitVariable.Float().Variable).Value
    In 1=VariableAccessor.Float(Variable=InitVariable.Float().Variable).Value
    In 2=VariableAccessor.Float(Variable=InitVariable.Float().Variable).Value
    In 3=VariableAccessor.Float(Variable=InitVariable.Float().Variable).Value
   )
   ArrayRandomGet_0 = Array.Random Get:WaveAsset:Array(In Array=^Sounds, Next=^UE.Source.OnPlay) {also {defaults {Enable Shared State="<MetaSoundBuilderResult.SUCCEEDED: 0>", No Repeats=-1, Weights []}}}
   WavePlayer_0 = UE.Wave Player:Mono(Play=@ArrayRandomGet_0.On Next, Wave Asset=@ArrayRandomGet_0.Value) {also {version [1,1]}}
   Multiply_0 = UE.Multiply:Audio by Float(
    AdditionalOperands=UE.Multiply:Float(PrimaryOperand=^MasterVolume, AdditionalOperands=@CrossfadeTriggerRouteFloat4_0.Out).Out
    PrimaryOperand=@WavePlayer_0.Out Mono
   )
  ]
  links [{dst="@UE.OutputFormat.Mono.Audio:0", src=Multiply_0.Out}, {dst="@UE.Source.OneShot.OnFinished", src=WavePlayer_0.On Finished}]
 }
]
properties {
 assetImportData=&self:AssetImportData
 attenuationSettings=&AS_Fort_WeaponFoley
 maxDistance=2250
 sampleRate=48000
 soundClassObject=&Fort_WeaponFoley
 concurrencySet [&CS_Fort_WeaponFoley]
}
